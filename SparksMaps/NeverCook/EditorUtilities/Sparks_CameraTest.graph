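# Editor utility widget that previews every unique camera Level Sequence in a chosen Sparks Camera Director Config, intended for Sparks_EmptyLevel.
# Play gathers shots from each section's Pools (not Intro/Exit) with their attach-to-actor flags.
# It spawns a player per sequence, setting transform origin from that flag and whether the config name contains "BattleStage".
# A short Delay loop polls playback and Next/Stop presses, destroying each actor before advancing.
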EditorUtilityWidgetBlueprint /SparksMaps/NeverCook/EditorUtilities/Sparks_CameraTest
graphs [
 Destroy previous sequence {nodes [
   Destroy previous sequence() {is=title K2Node_FunctionEntry}
   n1 = Get Current Sequence Actor() -> (Current Sequence Actor:Actor Object Reference)
   > n2 = DestroyActor!(execute, self:Editor Actor Subsystem Object Reference=Get EditorActorSubsystem(), ActorToDestroy:Actor Object Reference=@n1) -> (then, ReturnValue:bool)
   > PrintString(InString=Append!(A=Destroyed, B="  ", C:String=GetDisplayName(Object=@n1), D:String=" = ", E:String="To String (Boolean)"(InBool=@n2)) -> (ReturnValue))
   > Return Node()
  ]}
 EventGraph {nodes [
   "On Pressed (PlayButton)"() {is=title K2Node_ComponentBoundEvent}
   > Clear(TargetArray:Array of Level Sequence Object References=Get Level Sequence() -> (Level Sequence:Array of Level Sequence Object References))
   > Clear(TargetArray:Array of Booleans=Get Attach to Actor() -> (Attach to Actor:Array of Booleans))
   > n1 = For Each Loop(Array:Array of Sparks Camera Section Wrapper Structures=Get SequenceGenerators!(self:Sparks Camera Director Config Object Reference=Get Sparks Camera Director Config() -> (Sparks Camera Director Config:Sparks Camera Director Config Object Reference)) -> (SequenceGenerators:Array of Sparks Camera Section Wrapper Structures)) -> (Array Element:Sparks Camera Section Wrapper Structure)
   n2 = For Each Loop(Exec=@n1.Completed, Array:Array of Level Sequence Object References=Get Level Sequence()) -> (Array Element:Level Sequence Object Reference)
   SetVisibility(execute=@n2.Completed, self=Get PlayButton() -> (PlayButton:Editor Utility Button Object Reference), InVisibility=Collapsed)
   > SetVisibility(self=Get StopButton() -> (StopButton:Editor Utility Button Object Reference), InVisibility=Visible)
   > SetVisibility(self=Get NextCamera() -> (NextCamera:Editor Utility Button Object Reference), InVisibility=Visible)
   > n3 = SetVisibility(
    self=Get SparksCameraDirectorConfigPicker() -> (SparksCameraDirectorConfigPicker:Single Property View Object Reference)
    InVisibility=Collapsed
   )
   n4 = Get Level Sequence()
   n5 = Length(TargetArray:Array of Level Sequence Object References=@n4)
   n6 = Branch(execute=[@n15.then, @n3.then], Condition="integer <= integer"(A=@n5))
   > Stop Playing()
   n7 = "Get (a copy)"(Array:Array of Level Sequence Object References=@n4) -> (Output:Level Sequence Object Reference)
   n8 = CreateLevelSequencePlayer!(
    execute=@n6.else
    WorldContextObject:Object Reference
    LevelSequence:Level Sequence Object Reference=@n7
    Settings:Movie Scene Sequence Playback Settings Structure
   ) -> (then, OutActor:Level Sequence Actor Object Reference, ReturnValue:Level Sequence Player Object Reference)
   n9 = Get Attach to Actor()
   > Set Spawn Point!(
    execute
    self
    Level Sequence Actor:Level Sequence Actor Object Reference=@n8
    Level Sequence Name:String=GetDisplayName(Object=Get Sparks Camera Director Config())
    Is Attached to Actor:bool="Get (a copy)"(Array:Array of Booleans=@n9) -> (Output:bool)
   ) -> (then) {is=title K2Node_CallFunction}
   > Set Current Sequence Actor(Current Sequence Actor:Actor Object Reference=@n8) -> (Output_Get:Actor Object Reference)
   > n10 = Set Current Sequence Player(Current Sequence Player:Level Sequence Player Object Reference=@n8.ReturnValue) -> (Output_Get:Level Sequence Player Object Reference)
   > Play(self:Movie Scene Sequence Player Object Reference=@n10)
   > n11 = PrintString(InString=Append(A="Playing ", B=GetDisplayName(Object=@n7), D="Remaining:  ", E="To String (Integer)"(InInt=@n5)), Duration=6)
   n12 = Branch(execute=[@n17.then, @n11.then], Condition=IsPlaying!(self:Movie Scene Sequence Player Object Reference=@n10) -> (ReturnValue:bool))
   > n13 = Branch(Condition=Get Pressed Skip() -> (Pressed Skip:bool))
   > n14 = Set Pressed Skip(Pressed Skip:bool=false) -> (Output_Get:bool)
   Destroy previous sequence(execute=[@n14.then, @n12.else]) {is=title K2Node_CallFunction}
   > Remove Index!(execute, TargetArray:Array of Booleans=@n9, IndexToRemove:int=0) -> (then)
   > n15 = Remove Index(TargetArray:Array of Level Sequence Object References=@n4, IndexToRemove=0)
   n16 = Branch(execute=@n13.else, Condition=Get Pressed Stop() -> (Pressed Stop:bool))
   > Set Pressed Stop(Pressed Stop:bool=false) -> (Output_Get:bool)
   > Stop Playing()
   n17 = Delay!(execute=@n16.else, WorldContextObject:Object Reference, Duration=0.1) -> (then)
   PrintString(
    execute=@n2.LoopBody
    InString=Append!(A="To String (Integer)"(InInt=@n2.Array Index), B="   ", C=GetDisplayName(Object=@n2)) -> (ReturnValue)
   )
   n18 = For Each Loop(
    Exec=@n1.LoopBody
    Array:Array of Sparks Camera Pool Wrapper Structures=Break Sparks Camera Section Wrapper!(SparksCameraSectionWrapper:"Sparks Camera Section Wrapper Structure (by ref)"=@n1) -> (
     Name:Name
     Priority:int
     MaxGeneratedShotTimeSeconds:float=~
     PoolSelectionRule:Sparks Camera Pool Select Rule Base Class Reference
     IntroPool:Sparks Camera Pool Wrapper Structure
     Pools:Array of Sparks Camera Pool Wrapper Structures
     ExitPool:Sparks Camera Pool Wrapper Structure
     SectionFilter:Sparks Camera Filter Structure
    ) {is=title K2Node_BreakStruct}.Pools
   ) -> (Array Element:Sparks Camera Pool Wrapper Structure)
   > n19 = For Each Loop(Array:Array of Sparks Camera Shot Structures=Get Shots!(self:Sparks Camera Shot Pool Object Reference=Break Sparks Camera Pool Wrapper!(SparksCameraPoolWrapper:"Sparks Camera Pool Wrapper Structure (by ref)"=@n18) -> (ShotSelectionRule:Sparks Camera Shot Select Rule Base Class Reference, Pool:Sparks Camera Shot Pool Object Reference) {is=title K2Node_BreakStruct}.Pool) -> (Shots:Array of Sparks Camera Shot Structures)) -> (Array Element:Sparks Camera Shot Structure)
   n20 = Break Sparks Camera Shot!(SparksCameraShot:"Sparks Camera Shot Structure (by ref)"=@n19) -> (
    Sequence:Level Sequence Object Reference
    MoveToSocketName:Name
    bAttachToActor:bool=~
    bUseSocketRotationOnAttach:bool=~
    bCutOnMusicalBoundary:bool=~
    AttachInterpSpeed:float=~
    LookAtSocketName:Name
    LookAtInterpSpeed:float=~
    MinTimeSeconds:float=~
    MaxTimeSeconds:float=~
    Tags:Gameplay Tag Container Structure
   ) {is=title K2Node_BreakStruct}
   > n21 = Add Unique(TargetArray:Array of Level Sequence Object References=Get Level Sequence(), NewItem:"Level Sequence Object Reference (by ref)"=@n20)
   > Branch(Condition="integer >= integer"(A=@n21))
   > Add(TargetArray:Array of Booleans=Get Attach to Actor(), NewItem:"Boolean (by ref)"=@n20.bAttachToActor) {is=title K2Node_CallArrayFunction}
   "On Clicked (StopButton)"() {is=title K2Node_ComponentBoundEvent}
   > Set Pressed Stop(Pressed Stop=true)
   "On Clicked (NextCamera)"() {is=title K2Node_ComponentBoundEvent}
   > Set Pressed Skip(Pressed Skip=true)
   Event PreConstruct()
   > SetVisibility(self=Get StopButton(), InVisibility=Collapsed)
   > SetVisibility(self=Get NextCamera(), InVisibility=Collapsed)
   > SetObject(self=Get SparksCameraDirectorConfigPicker(), NewObject=Self-Reference())
   Event Destruct()
   > Stop Playing()
   Event Run()
   > RegisterTabAndGetID!(
    execute
    self:Editor Utility Subsystem Object Reference=Get EditorUtilitySubsystem()
    InBlueprint:Editor Utility Widget Blueprint Object Reference=&self
   ) -> (then, NewTabID:Name=None)
  ]}
 Set Spawn Point {nodes [
   n1 = Set Spawn Point!() -> (then, Level Sequence Actor:Level Sequence Actor Object Reference, Level Sequence Name:String, Is Attached to Actor:bool=~) {is=title K2Node_FunctionEntry}
   > Set bOverrideInstanceData!(execute, bOverrideInstanceData:bool=true, self:Level Sequence Actor Object Reference=@n1) -> (then, Output_Get:bool)
   > n2 = Cast To DefaultLevelSequenceInstanceData!(execute, Object=Get DefaultInstanceData!(self:Level Sequence Actor Object Reference=@n1) -> (DefaultInstanceData:Object Reference)) -> (then, CastFailed, AsDefault Level Sequence Instance Data:Default Level Sequence Instance Data Object Reference) {is=title K2Node_DynamicCast}
   > n3 = Branch(Condition=Contains(SearchIn=@n1.Level Sequence Name, Substring=BattleStage))
   > n4 = Branch(Condition=@n1.Is Attached to Actor)
   > n5 = Set TransformOrigin!(
    execute
    TransformOrigin:Transform=Get BattleStage Actor Location() -> (BattleStage Actor Location:Transform)
    self:Default Level Sequence Instance Data Object Reference=@n2
   ) -> (then, Output_Get:Transform)
   Return Node(execute=[@n8.then, @n5.then, @n6.then, @n9.then])
   n6 = Set TransformOrigin(execute=@n4.else, TransformOrigin=Get BattleStage Center() -> (BattleStage Center:Transform), self=@n2)
   n7 = Branch(execute=@n3.else, Condition=@n1.Is Attached to Actor)
   > n8 = Set TransformOrigin(TransformOrigin=Get MainStage Center() -> (MainStage Center:Transform), self=@n2)
   n9 = Set TransformOrigin(execute=@n7.else, TransformOrigin=Get Origin() -> (Origin:Transform), self=@n2)
  ]}
 Stop Playing {nodes [
   Stop Playing() {is=title K2Node_FunctionEntry}
   > SetVisibility(self=Get PlayButton(), InVisibility=Visible)
   > SetVisibility(self=Get StopButton(), InVisibility=Collapsed)
   > SetVisibility(self=Get NextCamera(), InVisibility=Collapsed)
   > Stop(self:Movie Scene Sequence Player Object Reference=Get Current Sequence Player() -> (Current Sequence Player:Level Sequence Player Object Reference))
   > DestroyActor(self=Get EditorActorSubsystem(), ActorToDestroy=Get Current Sequence Actor())
   > Clear(TargetArray:Array of Level Sequence Object References=Get Level Sequence())
   > Clear(TargetArray:Array of Booleans=Get Attach to Actor())
   > SetVisibility(self=Get SparksCameraDirectorConfigPicker(), InVisibility=Visible)
  ]}
]
variables [
 Attach to Actor {type="array<boolean>", default []}
 BattleStage Actor Location {type=Transform, default {location=<-53264 -51947 44728>, rotation=<rot 0 180 0>, scale=<1 1 1>}}
 BattleStage Center {type=Transform, default {location=<0 0 0>, rotation=<rot 0 0 0>, scale=<1 1 1>}}
 Current Sequence Actor {type=Actor, default=null}
 Current Sequence Player {type=LevelSequencePlayer, default=null}
 Level Sequence {type="array<LevelSequence>", default []}
 MainStage Center {type=Transform, default {location=<-6900 -460 720>, rotation=<rot 0 0 0>, scale=<1 1 1>}}
 NewVar {type=SparksCameraShotPool, default=null}
 Origin {type=Transform, default {location=<0 0 0>, rotation=<rot 0 0 0>, scale=<1 1 1>}}
 Pressed Skip {type=boolean}
 Pressed Stop {type=boolean}
 Sequence {type=MovieSceneSequence, default=null}
 Sequences {type="array<MovieSceneSequencePlayer>", default [null,null]}
 Sparks Camera Director Config {type=SparksCameraDirectorConfig, default=/PilgrimQuickplay/CameraDirector/Configs/PilgrimCamera_MainSequence_01}
]
widgets {
 info {rootWidgetClass=&StackBox, widgetCount=10}
 widgets [
  StackBox_0 {widgetClassPath=&StackBox, widget=StackBox_0, properties {orientation=Orient_Vertical}}
  AssetPicker {
   widgetClassPath=&VerticalBox
   parent=StackBox_0
   widget=AssetPicker
   slot {class=StackBoxSlot, properties {horizontalAlignment=HAlign_Center, verticalAlignment=VAlign_Center, padding=<ltrb 0 20 0 0>}}
  }
  SparksCameraDirectorConfigPicker {
   widgetClassPath=&SinglePropertyView
   parent=AssetPicker
   widget=SparksCameraDirectorConfigPicker
   properties {propertyName=Sparks Camera Director Config}
   slot {class=VerticalBoxSlot, properties {padding=<ltrb 20 10 20 50>}}
  }
  Actions {
   widgetClassPath=&VerticalBox
   parent=StackBox_0
   widget=Actions
   slot {class=StackBoxSlot, properties {horizontalAlignment=HAlign_Center, verticalAlignment=VAlign_Center}}
  }
  PlayButton {
   widgetClassPath=&EditorUtilityButton
   parent=Actions
   widget=PlayButton
   properties {toolTipText="Plays all unique camera sequences in the selected config.\r\n\r\nOnly works in \"Sparks_EmptyLevel\""}
   slot {
    class=VerticalBoxSlot
    properties {horizontalAlignment=HAlign_Center, verticalAlignment=VAlign_Center, padding=<ltrb 20 10 20 10>, size {sizeRule=Fill, value=1}}
   }
  }
  TextBlock_42 {
   widgetClassPath=&TextBlock
   parent=PlayButton
   widget=TextBlock_42
   properties {text="Play "}
   slot {class=ButtonSlot, properties {horizontalAlignment=HAlign_Fill, verticalAlignment=VAlign_Fill, padding=<ltrb 0 0 0 0>}}
  }
  NextCamera {
   widgetClassPath=&EditorUtilityButton
   parent=Actions
   widget=NextCamera
   slot {class=VerticalBoxSlot, properties {padding=<ltrb 20 10 20 10>}}
  }
  TextBlock_64 {widgetClassPath=&TextBlock, parent=NextCamera, widget=TextBlock_64, properties {text=Next Camera}, slot {class=ButtonSlot}}
  StopButton {
   widgetClassPath=&EditorUtilityButton
   parent=Actions
   widget=StopButton
   slot {
    class=VerticalBoxSlot
    properties {horizontalAlignment=HAlign_Center, verticalAlignment=VAlign_Center, padding=<ltrb 20 10 20 10>, size {sizeRule=Fill, value=1}}
   }
  }
  TextBlock_61 {widgetClassPath=&TextBlock, parent=StopButton, widget=TextBlock_61, properties {text="Stop "}, slot {class=ButtonSlot}}
 ]
}
